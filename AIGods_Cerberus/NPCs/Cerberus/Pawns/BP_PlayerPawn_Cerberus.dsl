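; Cerberus pawn Blueprint for an AI-controlled NPC that handles how the character appears when it spawns.
; On BeginPlay it calls the parent, then splits by network authority.
; Authority unhides the actor after a short delay as a safeguard.
; Remotes, if the pawn is hidden, fire a local burst GameplayCue for the lightning spawn-in VFX and unhide it immediately.
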(event EventBeginPlay
  (|CallParentFunction)
  (Networking|SwitchHasAuthority
    (:Authority
      ;[1 1 1] Make sure the AI shows up after 3s
      (Utilities|FlowControl|Delay 3.0)
      (Rendering|SetActorHiddenInGame false))
    (:Remote
      ;[1 1 1] Spawn in Lightning VFX and Show Pawn
      (if (Variables|Rendering|GetbHidden)
        (GameplayCue|ExecuteGameplayCue%28Burst%29Local self (Variables|Default|GetGC_SpawnIn) (Ability|GameplayCue|MakeGameplayCueParameters 0.0))
        (Rendering|SetActorHiddenInGame false)))))

(event Collision|EventActorBeginOverlap
  (params (OtherActor :type Actor%20Object%20Reference))
  (|CallParentFunction OtherActor) ; This node is disabled and will not be called.\nDrag off pins to build functionality.
)

(event EventTick
  (params (DeltaSeconds :type Float%20%28single-precision%29))
  (|CallParentFunction DeltaSeconds) ; This node is disabled and will not be called.\nDrag off pins to build functionality.
)

(fn ConstructionScript
  (|CallParentFunction))
