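; Cerberus NPC pawn variant for the Mash game mode that marks itself for players and drops perk loot when it dies.
; On BeginPlay it binds the death handler and, if flagged, registers itself as a sound indicator context for each player in the Mash mutator's data.
; On death it waits for playlist data, picks once from LTG_Perks, then tosses a copy of the first result per player in the game.
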(event EventBeginPlay
  (|CallParentFunction)
  (EventDispatchers|BindEventtoOnDied (delegate Custom|Died_SpawnLoot))
  (Default|MashMarkforPlayers))

(event Collision|EventActorBeginOverlap
  (params (OtherActor :type Actor%20Object%20Reference))
  (|CallParentFunction OtherActor) ; This node is disabled and will not be called.\nDrag off pins to build functionality.
)

(event Custom|Died_SpawnLoot
  (params (DamagedActor :type Actor%20Object%20Reference) (Damage :type Float%20%28single-precision%29) (InstigatedBy :type Controller%20Object%20Reference) (DamageCauser :type Actor%20Object%20Reference) (HitLocation :type Vector) (FHitComponent :type Primitive%20Component%20Object%20Reference) (BoneName :type Name) (Momentum :type Vector))
  (bind _gamestate (Class|FortAsyncActionCurrentPlaylistReady|PlaylistDataReadyAsync)
    (:OnReady
      (bind _outloottodrop (Loot|PickLootDrops "LTG_Perks" 0 -1 "(GameplayTags=)"))
      (for _ (range (Variables|/Script/FortniteGame.FortGameStateAthena|GetTotalPlayers _gamestate))
        (Inventory|SpawnPickupInWorldWithClassAndItemEntryForInstancedLoot (Utilities|Array|Get(aref) _outloottodrop) 0 (Transformation|GetActorLocation) "0.000000,0.000000,1.000000" 0 true true false "Tossed" "Unset" 0 false)))))

(event EventTick
  (params (DeltaSeconds :type Float%20%28single-precision%29))
  (|CallParentFunction DeltaSeconds) ; This node is disabled and will not be called.\nDrag off pins to build functionality.
)

(fn Mash_MarkForPlayers
  (if (Variables|Default|GetbMash_ShouldMarkForPlayers)
    (bind _asfort_game_state_athena (Utilities|Casting|CastToFortGameStateAthena (Game|GetGameState)))
    (bind _asmash_mutator (Utilities|Casting|CastToMashMutator (Modifier|GetMutatorbyClass (Variables|/Script/FortniteGame.FortGameStateAthena|GetMutatorListComponent _asfort_game_state_athena) "/Game/Athena/Playlists/Mash/MashMutator.MashMutator_C")))
    (bind _playerdataarray_entries (Utilities|Struct|BreakMashPlayerDataArray (Variables|/Script/FortniteGame.FortAthenaMutator_Mash|GetPlayerDataArray_Entries _asmash_mutator)))
    (for _array_element _playerdataarray_entries
      (Fortnite|RegisterSoundIndicatorPlacementContext (Fortnite|GetFortPlayerControllerfromActor (Utilities|Struct|BreakMashPlayerDataEntry _array_element)) self))))

(fn ConstructionScript
  (|CallParentFunction))
